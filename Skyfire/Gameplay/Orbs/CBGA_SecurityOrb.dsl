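; Skyfire loot orb that descends, lands and is collected by a long interaction, notifying its spawner and the Skyfire mutator.
; Projectile movement pauses while pawns overlap a collider, which clients destroy as only the server needs it.
; A replicated interaction flag drives Niagara parameters and a shrink, shake and glow timeline that reverses faster on release.
; Cosmetic events handle audio and FX; clients mark the orb after a delay for local players in the mothership.
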(event Custom|PlayOrbFxOnInteract
  (params (Is%20being%20interacted%20with :type Boolean))
  ;[1 1 1] Visual FX For On interaction
  (Components|Timeline|SetPlayRate (Variables|CBGA_SecurityOrb|GetInteractionTimeline) (Math|Float|SelectFloat 1.0 5.0 Is%20being%20interacted%20with)) ; Play 5 times faster in reverse when not being interacted with
  (if Is%20being%20interacted%20with
    (bind (_direction _value_0_1 _orbinnerglow) (Timeline|InteractionTimeline Play)
      (:Update
        (Transformation|SetWorldScale3D (Variables|Default|GetOrbMesh) (Math|Conversions|ToVector(Float) (Math|Float|Lerp 1.35 1.0 _value_0_1))) ; Shrink the orb
        (Rendering|Material|SetScalarParameterValue (Variables|Default|GetOrb%20Material) "InnerGlow" _orbinnerglow :!class /Script/Engine.MaterialInstanceDynamic) ; Make oth orb glow more
        (Transformation|SetRelativeLocation (Variables|Default|GetOrbMesh) (Math|Vector|Lerp(Vector) "0, 0, 0" (* (Math|Random|RandomUnitVector) 1.5) _value_0_1) false false) ; Shake The orb
      )) ; Do things On interact
    (else
      (Timeline|InteractionTimeline Reverse))))

(event EventBeginPlay
  ;[0 1 0.2825] EventBeginPlay - Set initial values
  (|CallParentFunction)
  (Utilities|FlowControl|Sequence
    (:then_0
      (Networking|SwitchHasAuthority
        (:Remote
          (Components|DestroyComponent (Variables|Default|GetPawnCollider)) ; Optimization - Destroy the PawnCollider component on the client as we only need it on the server
        )))
    (:then_1
      (Transformation|DetachFromComponent (Variables|Default|GetNS_Skyfire_LootOrb_Pos_Indicator) "KeepWorld" "KeepWorld" "KeepWorld" true) ; Detach the beam FX so they don't move
      (Networking|ForceNetUpdate)
      (Components|Movement|SetUpdatedComponent (Variables|Default|GetProjectileMovement) (Variables|Default|GetStaticMesh))
      (Components|Movement|SetUpdatedComponent (Variables|Default|GetRotatingMovement) (Variables|Default|GetOrbMesh))
      (CallFunction|InitializeLootOrbSounds)
      (Networking|SwitchHasAuthority
        (:Remote
          (Effects|Components|ParticleSystem|SpawnFXAtLocation (Variables|Default|GetSpawnFX) (Transformation|GetWorldLocation (Variables|Default|GetStaticMesh)))
          (Utilities|Time|SetTimerbyEvent (EventDispatchers|CreateEvent :!event Custom|CreateOrbMarkers) (ScalableFloat|GetScalableFloatMultiplier (Variables|Default|GetCreateOrbMarkerDelay))))))))

(event Custom|CreateOrbMarkers
  ;[0 1 0.5616] Create a marker for this orb for every local player controller.
  (for _array_element (Fortnite|GetLocalFortPlayerControllers true)
    (Utilities|IsValid (Game|Player|GetPlayerPawn _array_element)
      (:Is%20Valid
        (Utilities|IsValid (Actor|GetComponentbyClass (Game|Player|GetPlayerPawn _array_element) "/Script/SkyfireRuntime.SkyfirePlayerPawnComponent")
          (:Is%20Valid
            (if (Class|SkyfirePlayerPawnComponent|IsinMothership (Actor|GetComponentbyClass (Game|Player|GetPlayerPawn _array_element) "/Script/SkyfireRuntime.SkyfirePlayerPawnComponent"))
              (Class|SkyfirePlayerPawnComponent|MarkPlayspaceActoronClient (Actor|GetComponentbyClass (Game|Player|GetPlayerPawn _array_element) "/Script/SkyfireRuntime.SkyfirePlayerPawnComponent") self false) ; Applies ping marker to the orb, don't use detail hover (FORT-379461)
            )))))))

(event Interaction|EventOnGatherOrInteract
  (params (InteractingPawn :type Fort%20Pawn%20Object%20Reference) (InteractionBeingAttempted :type EInteractionBeingAttempted%20Enum :default "FirstInteraction"))
  ;[0.9583 0.3181 0] On Interact -Set is picked up, and notify the mutator
  (bind _spawner (Variables|Default|GetSpawner)
    (:then
      (Class|SkyfireVaultOrbSpawner|OnOrbPickedUp _spawner self)
      (bind _asfort_player_state_athena (Utilities|Casting|CastToFortPlayerStateAthena (Variables|/Script/Engine.Pawn|GetPlayerState InteractingPawn)))
      (bind _asfort_athena_mutator_skyfire (Utilities|Casting|CastToFortAthenaMutator_Skyfire (Modifier|GetMutatorbyClass (Variables|/Script/FortniteGame.FortGameStateAthena|GetMutatorListComponent (Fortnite|Athena|GameState|GetGameStateAthena)) "/Script/SkyfireRuntime.FortAthenaMutator_Skyfire")))
      (Class|FortAthenaMutatorSkyfire|OnLootRarityPickup _asfort_athena_mutator_skyfire _asfort_player_state_athena)
      (CallFunction|PlayCollectFX InteractingPawn))
    (:else)))

(event OnProjectileStop(ProjectileMovement)
  (params (ImpactResult_bBlockingHit :type Boolean) (ImpactResult_bInitialOverlap :type Boolean) (ImpactResult_Time :type Float%20%28single-precision%29) (ImpactResult_Distance :type Float%20%28single-precision%29) (ImpactResult_Location :type Vector :default "0, 0, 0") (ImpactResult_ImpactPoint :type Vector :default "0, 0, 0") (ImpactResult_Normal :type Vector :default "0, 0, 0") (ImpactResult_ImpactNormal :type Vector :default "0, 0, 0") (ImpactResult_PhysMat :type Physical%20Material%20Object%20Reference) (ImpactResult_HitActor :type Actor%20Object%20Reference) (ImpactResult_HitComponent :type Primitive%20Component%20Object%20Reference) (ImpactResult_HitBoneName :type Name :default "None") (ImpactResult_BoneName :type Name :default "None") (ImpactResult_HitItem :type Integer) (ImpactResult_ElementIndex :type Integer) (ImpactResult_FaceIndex :type Integer) (ImpactResult_TraceStart :type Vector :default "0, 0, 0") (ImpactResult_TraceEnd :type Vector :default "0, 0, 0"))
  ;[0.2578 0.1939 1] If stopped by anything other than a player, the orb has landed. Play landing sounds.
  (Utilities|Casting|CastToFortPlayerPawn ImpactResult_HitActor
    (:CastFailed
      (if (Utilities|NotEqual(Object) ImpactResult_HitActor self)
        (Utilities|FlowControl|DoOnce false
          (:Completed
            (CallFunction|LootOrbLandedSounds)
            (Variables|Default|SetbHasLanded true)))))))

(event Custom|PlayCollectFX
  (params (Player :type Fort%20Pawn%20Object%20Reference))
  (bind _islocalpawn (HelperFunctions|IsLocalPawn Player))
  ;[1 0.1569 0.7639] Play Collect Sound, Stop Looping Sound
  (Networking|CanExecuteCosmeticEvents
    (:True
      (Audio|Components|Audio|FadeOut (Variables|Sound|GetLoot_Orb_Idle_Loop%20AC) (Variables|Sound|GetLoot%20Orb%20Idle%20Fade%20Out%20Duration) 0.0 "Sin" :!class /Script/Engine.AudioComponent)
      (Utilities|FlowControl|DoOnce false
        (:Completed
          (Audio|SpawnSoundAttached (select _islocalpawn (Variables|Sound|GetSFX_LootOrb_Collected_P3) (Variables|Sound|GetSFX_LootOrb_Collected_P1)) (Variables|Default|GetStaticMesh))
          (Effects|Components|ParticleSystem|SpawnFXAtLocation (Variables|Default|GetPickupFX) (Transformation|GetWorldLocation (Variables|Default|GetStaticMesh)))
          (if _islocalpawn
            (Utilities|FlowControl|Delay (Variables|Sound|GetSFX_Alien_VO_Delay))
            (Audio|PlaySound2D (Variables|Sound|GetSFX_Alien_VO))))))))

(event Interaction|EventOnServerNotifyStartLongUse
  (params (InteractingPawn :type Fort%20Player%20Pawn%20Object%20Reference))
  ;[0 0.3378 1] Set Interact bool based on interact events. Play/Stop Interacting sound
  (Variables|Default|SetbIsBeingInteractedWith true)
  ;;[1 0.2146 0.9942] Cosmetic Events
  (CallFunction|PlayInteractSound InteractingPawn))

(event Interaction|EventOnServerNotifyEndLongUse
  (params (InteractingPawn :type Fort%20Player%20Pawn%20Object%20Reference) (bUseCompleted :type Boolean))
  ;[0 0.3378 1] Set Interact bool based on interact events. Play/Stop Interacting sound
  (Variables|Default|SetbIsBeingInteractedWith false)
  ;;[1 0.2146 0.9942] Cosmetic Events
  (CallFunction|StopInteractSound))

(event Custom|InitializeLootOrbSounds
  ;[1 0.1749 0.8078] Play Idle Loop and descending Loop
  (Networking|CanExecuteCosmeticEvents
    (:True
      (Audio|Components|Audio|Play (Variables|Sound|GetLoot_Orb_Spawn%20AC) 0.0) ; Play Spawn Sound
      (Audio|Components|Audio|Play (Variables|Default|GetLoot_Orb_Descent_AC) 0.0) ; Play Descending Sound
    )))

(event Custom|LootOrbLandedSounds
  ;[1 0.1749 0.8078] Play Idle Loop and descending Loop
  (Networking|CanExecuteCosmeticEvents
    (:True
      (Audio|Components|Audio|Play (Variables|Sound|GetLoot_Orb_Idle_Loop%20AC) (Math|Random|RandomFloatinRange (Variables|Default|GetMin%20Loop%20Sound%20Start%20Time) (Variables|Default|GetMax%20Loop%20Sound%20Start%20Time))) ; Play Idle sound
      (Audio|Components|Audio|FadeOut (Variables|Default|GetLoot_Orb_Descent_AC) (Variables|Sound|GetDescending%20Loop%20Fade%20Out%20Duration) 0.0 "Linear" :!class /Script/Engine.AudioComponent) ; Fade Out Descending sound
    )))

(event Custom|PlayInteractSound
  (params (Player :type Fort%20Player%20Pawn%20Object%20Reference%20%28by%20ref%29))
  ;[1 0.2856 0.8636] Play & Stop Interacting Sound
  (Networking|CanExecuteCosmeticEvents
    (:True
      (bind _returnvalue (Audio|SpawnSoundatLocation (select (HelperFunctions|IsLocalPawn Player) (Variables|Sound|GetSFX_LootOrb_Collecting_P1) (Variables|Sound|GetSFX_LootOrb_Collecting_P3)) (Transformation|GetWorldLocation (Variables|Default|GetStaticMesh))))
      (Variables|Sound|SetSFX_CollectingOrb_AC _returnvalue))))

(event Custom|StopInteractSound
  ;[1 0.2856 0.8636] Play & Stop Interacting Sound
  (Networking|CanExecuteCosmeticEvents
    (:True
      (Audio|Components|Audio|FadeOut (Variables|Sound|GetSFX_CollectingOrb_AC) (Variables|Sound|GetSFX_CollectOrb_Fadeout_Duration) 0.0 "Linear" :!class /Script/Engine.AudioComponent))))

(event OnComponentBeginOverlap(PawnCollider)
  (params (OverlappedComponent :type Primitive%20Component%20Object%20Reference) (OtherActor :type Actor%20Object%20Reference) (OtherComp :type Primitive%20Component%20Object%20Reference) (OtherBodyIndex :type Integer) (bFromSweep :type Boolean) (SweepResult :type Hit%20Result%20Structure%20%28by%20ref%29))
  ;[0.7955 0.4314 1] Stop moving if we collide with a pawn, and begin moving when no longer colliding with any.
  ;;[0.09399 0.3275 1] If the orb has not yet landed and begins an overlap with a player, stop moving immediately.
  (if (not (Variables|Default|GetbHasLanded))
    (Utilities|Casting|CastToFortPawn OtherActor)
    (Game|Components|ProjectileMovement|StopSimulating (Variables|Default|GetProjectileMovement) SweepResult)))

(event OnComponentEndOverlap(PawnCollider)
  (params (OverlappedComponent :type Primitive%20Component%20Object%20Reference) (OtherActor :type Actor%20Object%20Reference) (OtherComp :type Primitive%20Component%20Object%20Reference) (OtherBodyIndex :type Integer))
  ;[0.7955 0.4314 1] Stop moving if we collide with a pawn, and begin moving when no longer colliding with any.
  ;;[0.1455 0.3732 1] If the orb has not yet landed and a player ends overlap with the orb, try to start moving again as long as there are no other pawns overlapping.
  (if (not (Variables|Default|GetbHasLanded))
    (Utilities|Casting|CastToFortPawn OtherActor)
    (if (<= (Utilities|Array|Length (Collision|GetOverlappingActors (Variables|Default|GetPawnCollider) "/Script/FortniteGame.FortPlayerPawn" :!class /Script/Engine.PrimitiveComponent)) 0)
      (Components|Movement|SetUpdatedComponent (Variables|Default|GetProjectileMovement) (Variables|Default|GetStaticMesh)))))

(fn GetDeathFXTransform
  (return (Transformation|GetWorldTransform (Variables|Default|GetStaticMesh))))

(fn OnRep_bIsBeingInteractedWith
  (bind _bisbeinginteractedwith (Variables|Default|GetbIsBeingInteractedWith))
  (bind _interactparameter (Variables|Default|GetInteractParameter))
  ;[0.5789 1 0] Set the parameters when the bool is updated.
  (Effects|Components|ParticleSystem|SetBoolParameter (Variables|Default|GetNS_LootOrb_Idle) _interactparameter _bisbeinginteractedwith)
  (Effects|Components|ParticleSystem|SetBoolParameter (Variables|Default|GetNS_Skyfire_LootOrb_Pos_Indicator) _interactparameter _bisbeinginteractedwith)
  ;[1 1 1] play timeline to update visuas
  (CallFunction|PlayOrbFxonInteract _bisbeinginteractedwith))

(fn ConstructionScript
  ;[1 1 1] Visual FX For On interaction (For use with PlayOrbFxOnInteract)
  (bind _returnvalue (Rendering|Material|CreateDynamicMaterialInstance (Variables|Default|GetOrbMesh) 0 (Rendering|Material|GetMaterial (Variables|Default|GetOrbMesh) 0) "OrbMaterial" :!class /Script/Engine.PrimitiveComponent))
  (Variables|Default|SetOrb%20Material _returnvalue))
