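; Security orb pickup, built on FortPickup, that spins in place and is collected when a Fortnite player pawn overlaps it.
; BeginPlay points projectile movement at the collision box, disables bouncing and applies a configurable gravity scale.
; On authority, overlap notifies its SecurityOrbSpawner and sets the pawn as pickup target.
; Remote instances play the pickup sound on any overlap, while Tick rotates the mesh by YawPerSecond.
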(event EventBeginPlay
  (bind _movementcomponent (Variables|FortPickup|GetMovementComponent))
  ;[0 1 0.2825] EventBeginPlay - Set initial values
  (|CallParentFunction)
  (Networking|ForceNetUpdate)
  (Components|Movement|SetUpdatedComponent _movementcomponent (Variables|Default|GetCollisionBox))
  (Variables|/Script/Engine.ProjectileMovementComponent|SetbShouldBounce false _movementcomponent)
  (Variables|/Script/Engine.ProjectileMovementComponent|SetProjectileGravityScale (Variables|Default|GetProjectile%20Gravity%20Scale) _movementcomponent))

(event OnComponentBeginOverlap(CollisionBox)
  (params (OverlappedComponent :type Primitive%20Component%20Object%20Reference) (OtherActor :type Actor%20Object%20Reference) (OtherComp :type Primitive%20Component%20Object%20Reference) (OtherBodyIndex :type Integer) (bFromSweep :type Boolean) (SweepResult :type Hit%20Result%20Structure%20%28by%20ref%29))
  ;[1 0.9192 0] OnComponentBeginOverlap - if overlapping with a player, play sound effect, set is picked up, add  and destory the actor
  (Networking|SwitchHasAuthority
    (:Authority
      (bind _asfort_player_pawn_athena (Utilities|Casting|CastToFortPlayerPawnAthena OtherActor))
      (Utilities|Casting|CastToFortPlayerControllerAthena (Pawn|GetController _asfort_player_pawn_athena))
      (bind _spawner (Variables|Default|GetSpawner)
        (:then
          (Class|SecurityOrbSpawner|OnOrbPickedUp _spawner self)
          (Pickup|BlueprintSetPickupTarget _asfort_player_pawn_athena))
        (:else)))
    (:Remote
      (Audio|PlaySoundatLocation "/Game/Sounds/BattleLab/BotSnack/BotSnack_Pickup_Cue.BotSnack_Pickup_Cue" (Transformation|GetActorLocation)))))

(event EventTick
  (params (DeltaSeconds :type Float%20%28single-precision%29))
  ;[0.2361 0.9303 1] Event Tick, have the orb spin in place
  (|CallParentFunction DeltaSeconds)
  (Transformation|AddLocalRotation (Variables|Default|GetStaticMesh) (Math|Rotator|MakeRotator 0.0 0.0 (* DeltaSeconds (Variables|Default|GetYawPerSecond))) false false))

(fn ConstructionScript)
