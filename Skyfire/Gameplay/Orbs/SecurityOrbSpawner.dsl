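; Skyfire security orb spawner that, on the CoreGameplay phase, spawns an initial delayed batch of orbs at tagged actor locations on the authority.
; It waits for the Skyfire mutator, reads playlist-specific tuning, and reports orb count changes to it.
; Timers refill orbs by a shortfall-scaled random roll and step the maximum through levels.
; A map returns picked-up orbs' points to the pool; clients play a matching delayed spawn sound.
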(event EventBeginPlay
  ;[0 1 0.4359] Event Begin Play - Set Initial values based off of scalable floats. Start waiting for mutator
  (|CallParentFunction)
  (Variables|Default|SetSecurityOrbSpawnerOffset (Math|Vector|MakeVector 0.0 0.0 (ScalableFloat|GetValueatLevel (Variables|Default|GetSFSpawnDistanceZOffset))))
  (Variables|Default|SetMaxNumberOfSecurityOrbsAtOnce (Math|Float|Truncate (ScalableFloat|GetValueatLevel (Variables|Default|GetSFMaxSecurityOrbsAtOnce))))
  (Variables|Default|SetPercentageChanceToSpawnNewOrbPerOrbsFromMax (ScalableFloat|GetValueatLevel (Variables|Default|GetSFPrecentageChanceToSpawnNewOrbPerOrbsFromMax)))
  (CallFunction|WaitforMutator))

(event Custom|WaitForMutator
  (bind _returnvalue (Fortnite|Athena|GameState|GetGameStateAthena))
  ;[0.142 0.8701 1] WaitForMutator - Wait for and bind to the skyfire mutator
  (Utilities|IsValid _returnvalue
    (:Is%20Valid
      (bind _asfort_athena_mutator_skyfire (Utilities|Casting|CastToFortAthenaMutator_Skyfire (Modifier|GetMutatorbyClass (Variables|/Script/FortniteGame.FortGameStateAthena|GetMutatorListComponent _returnvalue) "/Script/SkyfireRuntime.FortAthenaMutator_Skyfire"))
        (:then
          (bind _output_get (Variables|Default|SetSkyfireMutator _asfort_athena_mutator_skyfire))
          (EventDispatchers|BindEventtoOnSkyfirePhaseChanged _output_get (EventDispatchers|CreateEvent :!event Custom|OnSkyfirePhaseChanged))
          (CallFunction|DeterminePlaylistType)
          (CallFunction|SetOrbSpawnValuesBasedonPlaylist))
        (:CastFailed
          (Utilities|Time|SetTimerbyEvent (delegate Custom|WaitForMutator) 0.1) ; Wait and retry
        )))))

(event Custom|SpawnInitalOrbs
  (bind _maxnumberofsecurityorbsatonce (Variables|Default|GetMaxNumberOfSecurityOrbsAtOnce))
  ;[1 0.9206 0.202] SpawnInitialOrbs - After an initial delay, spawn the max number of orbs available and then start checking for orbs.
  (Networking|SwitchHasAuthority
    (:Authority
      (Utilities|FlowControl|Delay (ScalableFloat|GetValueatLevel (Variables|Default|GetInitialDelayToSpawn)))
      (for _ (range _maxnumberofsecurityorbsatonce)
        (CallFunction|SpawnSecurityOrb))
      (CallFunction|UpdateOrbCount _maxnumberofsecurityorbsatonce)
      (bind _returnvalue (Utilities|Time|SetTimerbyEvent (EventDispatchers|CreateEvent :!event Custom|CheckForTotalOrbs) (Variables|Default|GetTimeToCheckTotalOrbs) true))
      (Variables|Default|SetCheckForTotalOrbsHandle _returnvalue)
      (Utilities|Time|SetTimerbyEvent (EventDispatchers|CreateEvent :!event Custom|UpdateMaxOrbLimit) (ScalableFloat|GetValueatLevel (Variables|Default|GetSF_MaxOrbLimitTimers))))))

(event Custom|OnSkyfirePhaseChanged
  (params (NewPhase :type ESkyfirePhase%20Enum))
  (bind _outputdelegate (EventDispatchers|CreateEvent :!event Custom|OnSkyfirePhaseChanged))
  ;[0.5472 1 0.1076] OnSkyfirePhaseChanged - On phase CoreGameplay, set spawn points and spawn initial orbs
  (Networking|SwitchHasAuthority
    (:Authority
      (switch Utilities|FlowControl|Switch|SwitchonESkyfirePhase NewPhase
        (:CoreGameplay
          (CallFunction|SetSpawnPoints)
          (CallFunction|SpawnInitalOrbs)
          (bind _skyfiremutator (Variables|Default|GetSkyfireMutator)
            (:then
              (EventDispatchers|UnbindEventfromOnSkyfirePhaseChanged _skyfiremutator _outputdelegate))
            (:else)))))
    (:Remote
      (switch Utilities|FlowControl|Switch|SwitchonESkyfirePhase NewPhase
        (:CoreGameplay
          (Utilities|FlowControl|Delay (ScalableFloat|GetValueatLevel (Variables|Default|GetInitialDelayToSpawn))) ; Spawn Init Orbs has a delay, we need to duplicate it
          (CallFunction|TrytoPlayOrbSpawnSound)
          (bind _skyfiremutator_2 (Variables|Default|GetSkyfireMutator)
            (:then
              (EventDispatchers|UnbindEventfromOnSkyfirePhaseChanged _skyfiremutator_2 _outputdelegate))
            (:else)))))))

(event Custom|CheckForTotalOrbs
  ;[0.1199 0.1137 1] CheckForTotalOrbs - If there are less orbs that the max, check if we should spawn an orb and spawn orb if we should
  (if (> (- (Variables|Default|GetMaxNumberOfSecurityOrbsAtOnce) (Utilities|Map|Length (Variables|Default|GetOrbToSpawnLocationMap))) 0)
    (if (CallFunction|ShouldSpawnNewOrb)
      (CallFunction|SpawnSecurityOrb)
      (CallFunction|UpdateOrbCount 1))))

(event Custom|UpdateMaxOrbLimit
  (bind _sf_maxorblimittimers (Variables|Default|GetSF_MaxOrbLimitTimers))
  ;[0.9504 0.3663 1] Increment the MaxOrbLevelLimit and update the MaxOrbs based on the current level.
  (bind _output_get (Variables|Default|SetMaxOrbLimitLevel (+ (Variables|Default|GetMaxOrbLimitLevel) 1)))
  (Variables|Default|SetMaxNumberOfSecurityOrbsAtOnce (Math|Float|Truncate (ScalableFloat|GetValueatLevel (Variables|Default|GetSF_MaxOrbs) (Math|Conversions|ToFloat(Integer) _output_get))))
  (if (not (< (ScalableFloat|GetValueatLevel _sf_maxorblimittimers (Math|Conversions|ToFloat(Integer) _output_get)) 0.0))
    (Utilities|Time|SetTimerbyEvent (delegate Custom|UpdateMaxOrbLimit) (ScalableFloat|GetValueatLevel _sf_maxorblimittimers (Math|Conversions|ToFloat(Integer) _output_get)))))

(fn DeterminePlaylistType
  (bind _returnvalue (Fortnite|Athena|GameState|GetGameStateAthena))
  ;[0.8353 1 0] Get the playlist tag to determine if it's solo/squad etc.
  (Utilities|IsValid _returnvalue
    (:Is%20Valid
      (bind _array_element (Utilities|Array|ForEachLoopwithBreak (Variables|Default|GetPossiblePlaylistTags))
        (:LoopBody
          (if (GameplayTags|TagContainer|CompareTagContainertoOtherTags _array_element (Class|FortGameStateAthena|GetAthenaPlaylistContextTags _returnvalue))
            (Variables|Default|SetT_PlaylistType _array_element)))))))

(fn GetRandomSpawnPoint
  ;[0.3488 0.9202 1] GetRandomSpawnPoint - Get a random spawn point to spawn a orb
  (bind _result (ArrayMacros|SafeArrayGet (Variables|Default|GetAvailableSpawnPoints) (Math|Random|RandomInteger (Utilities|Array|Length (Variables|Default|GetAvailableSpawnPoints))))
    (:%20%20
      (return _result))))

(fn OnOrbPickedUp
  (params (SecurityOrb :type Actor%20Object%20Reference))
  (bind _orbtospawnlocationmap (Variables|Default|GetOrbToSpawnLocationMap))
  (bind (_value _returnvalue) (Utilities|Map|Find _orbtospawnlocationmap SecurityOrb))
  ;[0.03035 1 0.3644] OnOrbPickedUp - Remove the orb from the TMap and it's location from Used Spawn Locations
  (CallFunction|UpdateOrbCount -1)
  (if _returnvalue
    (Utilities|Array|Add (Variables|Default|GetAvailableSpawnPoints) _value)
    (Utilities|Map|Remove _orbtospawnlocationmap SecurityOrb)))

(fn SetOrbSpawnValuesBasedOnPlaylist
  (bind (_value _returnvalue) (Utilities|Map|Find (Variables|Default|GetPlaylistTag_MaxOrbs_Map) (Variables|Default|GetT_PlaylistType)))
  (bind (_value_2 _returnvalue_2) (Utilities|Map|Find (Variables|Default|GetPlaylistTag_TimeToCheckOrbs_Map) (Variables|Default|GetT_PlaylistType)))
  (bind (_value_3 _returnvalue_3) (Utilities|Map|Find (Variables|Default|GetPlaylistTag_OrbLimitTimers_Map) (Variables|Default|GetT_PlaylistType)))
  ;[0 0.5426 1] Set various Orb Spawn values based on the playlist tag/type.
  (Utilities|FlowControl|Sequence
    (:then_0
      (Variables|Default|SetTimeToCheckTotalOrbs (ScalableFloat|GetValueatLevel (select _returnvalue_2 _value_2 (Variables|Default|GetSFTimeToCheckTotalOrbs)))))
    (:then_1
      (Variables|Default|SetSF_MaxOrbs (select _returnvalue _value (Variables|Default|GetSF_MaxOrbs)))
      (Variables|Default|SetSF_MaxOrbLimitTimers (select _returnvalue_3 _value_3 (Variables|Default|GetSF_MaxOrbLimitTimers))))))

(fn SetSpawnPoints
  ;[0.731 0.3625 1] Set Spawn Points - Set the spawn points for Security Orbs
  (bind _outactors (Actor|GetAllActorswithTag (Variables|Default|GetSecurityOrbTag)))
  (for _array_element _outactors
    (Utilities|Array|Add (Variables|Default|GetSecurityOrbsSpawnPoints) (.location (Transformation|GetActorTransform _array_element))))
  (Variables|Default|SetAvailableSpawnPoints (Variables|Default|GetSecurityOrbsSpawnPoints)))

(fn ShouldSpawnNewOrb
  (bind _orbtospawnlocationmap (Variables|Default|GetOrbToSpawnLocationMap))
  (bind _returnvalue (Utilities|Map|Length _orbtospawnlocationmap))
  ;[0.381 1 0.1435] ShouldSpawnOrb, return whether or not we should spawn an orb based on percentage chance based on how many orbs from the max there currently are, if no orbs always spawn one.
  (if (> _returnvalue 0)
    (return (>= (Math|Random|RandomFloat) (* (- (Variables|Default|GetMaxNumberOfSecurityOrbsAtOnce) _returnvalue) (Variables|Default|GetPercentageChanceToSpawnNewOrbPerOrbsFromMax))))
    (else
      (return true))))

(fn SpawnSecurityOrb
  ;[0 0.2438 1] SpawnSecurityOrb - Spawn a security orb and add it to the TMap as well as its location to the Used Spawn Locations
  (bind _output_get (LocalVariables|SetFoundSpawnPoint (CallFunction|GetRandomSpawnPoint))) ; Store the randomly found point
  (bind _returnvalue (Game|SpawnActorfromClass "/Skyfire/Gameplay/Orbs/CBGA_SecurityOrb.CBGA_SecurityOrb_C" (Math|Transform|MakeTransform (+ _output_get (Variables|Default|GetSecurityOrbSpawnerOffset))) "AlwaysSpawn" "OverrideRootScale" 0 0 0))
  (Utilities|Map|Add (Variables|Default|GetOrbToSpawnLocationMap) _returnvalue _output_get)
  (Utilities|Array|RemoveItem (Variables|Default|GetAvailableSpawnPoints) _output_get))

(fn TryToPlayOrbSpawnSound
  (bind _returnvalue (Fortnite|GetFirstLocalFortPlayerController true true))
  (bind _returnvalue_2 (Game|Player|GetPlayerPawn _returnvalue))
  (bind _returnvalue_3 (Actor|GetComponentbyClass _returnvalue_2 "/Script/SkyfireRuntime.SkyfirePlayerPawnComponent"))
  (Utilities|IsValid _returnvalue_2
    (:Is%20Valid
      (Utilities|IsValid _returnvalue_3
        (:Is%20Valid
          (if (Class|SkyfirePlayerPawnComponent|IsinMothership _returnvalue_3)
            (Audio|PlaySound2D (Variables|Sound|GetSFX_LootOrb_Appear_FirstTime))
            (Audio|PlaySound2D (Variables|Sound|GetSFX_VO_LootOrb_Appear_FirstTime))))))))

(fn UpdateOrbCount
  (params (Amount :type Integer))
  (bind _skyfiremutator (Variables|Default|GetSkyfireMutator)
    (:then
      (Playspace|ModifySecurityOrbCount _skyfiremutator Amount))
    (:else)))

(fn ConstructionScript)
