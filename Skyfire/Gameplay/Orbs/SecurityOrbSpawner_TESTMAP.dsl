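; Test-map variant of the security orb spawner, intended to keep one orb constantly respawning.
; Its EventWaitForMutator event determines the playlist type, applies playlist-based orb spawn values, sets spawn points and spawns the initial orbs, in that order.
; The called functions are not defined in this file, and the construction script is empty.
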(event EventWaitForMutator
  ;[0 1 0.1595] Handle spawning in the test map. Set up spawn values so one orb will constantly respawn.
  (CallFunction|DeterminePlaylistType)
  (CallFunction|SetOrbSpawnValuesBasedonPlaylist)
  (CallFunction|SetSpawnPoints)
  (CallFunction|SpawnInitalOrbs))

(fn ConstructionScript)
