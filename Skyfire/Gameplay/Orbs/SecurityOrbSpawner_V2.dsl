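; Blueprint security orb spawner for Skyfire that links spawned orbs to itself and plays first-appearance audio.
; On each orb spawn, it casts to CBGA_SecurityOrb and sets itself as Spawner, done here because the orb class is Blueprint-only.
; When initial spawning completes, it plays one of two client-side 2D sounds, SFX or VO, depending on whether the local player's SkyfirePlayerPawnComponent reports being in the mothership.
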(event EventOnInitialOrbSpawnComplete
  ;[0.1857 1 0.1834] Attempt to play sounds on the client after the initial orbs have been spawned.
  (CallFunction|TryPlayOrbSpawnSound))

(event EventOnOrbSpawned
  (params (SpawnedOrb :type Actor%20Object%20Reference))
  ;[0.716 1 0.2325] This needs to be set in blueprints because the SecurityOrb class is BP only.
  (bind _ascbga_security_orb (Utilities|Casting|CastToCBGA_SecurityOrb SpawnedOrb))
  (Variables|/Skyfire/Gameplay/Orbs/CBGA_SecurityOrb.CBGA_SecurityOrb_C|SetSpawner self _ascbga_security_orb))

(fn TryPlayOrbSpawnSound
  (bind _returnvalue (Fortnite|GetFirstLocalFortPlayerController true true))
  (bind _returnvalue_2 (Game|Player|GetPlayerPawn _returnvalue))
  (bind _returnvalue_3 (Actor|GetComponentbyClass _returnvalue_2 "/Script/SkyfireRuntime.SkyfirePlayerPawnComponent"))
  (Utilities|IsValid _returnvalue_2
    (:Is%20Valid
      (Utilities|IsValid _returnvalue_3
        (:Is%20Valid
          (if (Class|SkyfirePlayerPawnComponent|IsinMothership _returnvalue_3)
            (Audio|PlaySound2D (Variables|Sound|GetSFX_LootOrb_Appear_FirstTime))
            (Audio|PlaySound2D (Variables|Sound|GetSFX_VO_LootOrb_Appear_FirstTime))))))))

(fn ConstructionScript)
