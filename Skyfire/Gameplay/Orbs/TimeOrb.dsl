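; Time orb pickup for the Skyfire mode that players collect by touch, with spawn, idle loop and expiry FX.
; On overlap, the server notifies the Skyfire mutator and TimerOrbSpawner, sets the pickup target and replicates the collector by waking net dormancy.
; The collector's OnRep hides the orb and plays first- or third-person FX.
; Local clients predict only the pickup FX, keeping the orb visible until server confirmation.
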(event EventBeginPlay
  ;[0.01702 1 0.3402] EventBeginPlay
  (|CallParentFunction)
  (Components|Activation|SetActive (Variables|FortPickup|GetMovementComponent) false false)
  (Game|BindEventtoOnDestroyed (EventDispatchers|CreateEvent :!event Custom|TimeOrbDestroyed))
  (Audio|Components|Audio|Play (Variables|Sound|GetTimeOrb_Idle_Loop_Comp) (Math|Random|RandomFloatinRange (Variables|Sound|GetMin%20Loop%20Sound%20Start%20Time) (Variables|Sound|GetMax%20Loop%20Sound%20Start%20Time)))
  (Effects|Components|ParticleSystem|SpawnFXAtLocation "/Skyfire/Gameplay/Orbs/FX/Time/NS_TimeOrb_Spawn.NS_TimeOrb_Spawn" (Transformation|GetActorLocation)))

(event OnComponentBeginOverlap(TouchCapsule)
  (params (OverlappedComponent :type Primitive%20Component%20Object%20Reference) (OtherActor :type Actor%20Object%20Reference) (OtherComp :type Primitive%20Component%20Object%20Reference) (OtherBodyIndex :type Integer) (bFromSweep :type Boolean) (SweepResult :type Hit%20Result%20Structure%20%28by%20ref%29))
  (bind _returnvalue (Fortnite|Athena|GameState|GetGameStateAthena))
  (bind _mutatorlistcomponent (Variables|/Script/FortniteGame.FortGameStateAthena|GetMutatorListComponent _returnvalue))
  ;[1 0.904 0.02277] OnComponentBeginOverlap - if overlapping with a player, set them as the pickup target and collect the pickup
  (bind _asfort_player_pawn_athena (Utilities|Casting|CastToFortPlayerPawnAthena OtherActor))
  (if (not (Variables|Default|GetbTouched))
    (Networking|SwitchHasAuthority
      (:Authority
        (bind _asfort_player_state_athena (Utilities|Casting|CastToFortPlayerStateAthena (Variables|/Script/Engine.Pawn|GetPlayerState _asfort_player_pawn_athena)))
        (Utilities|IsValid _returnvalue
          (:Is%20Valid
            (Utilities|IsValid _mutatorlistcomponent
              (:Is%20Valid
                (bind _asfort_athena_mutator_skyfire (Utilities|Casting|CastToFortAthenaMutator_Skyfire (Modifier|GetMutatorbyClass _mutatorlistcomponent "/Script/SkyfireRuntime.FortAthenaMutator_Skyfire")))
                (Class|FortAthenaMutatorSkyfire|OnTimePickup _asfort_athena_mutator_skyfire _asfort_player_state_athena)
                (Variables|Default|SetbTouched true)
                (Pickup|BlueprintSetPickupTarget _asfort_player_pawn_athena)
                (bind _timerorbspawner (Variables|Default|GetTimerOrbSpawner)
                  (:then
                    (Class|TimerOrbSpawner|OnTimeOrbPickedUp _timerorbspawner self))
                  (:else
                    _))
                ;;[0.9197 0.4526 1] Replicate the player pawn who collected this pickup for FX purposes. Pickups are Dormant by default, so kick them to get this replicated. (Don't need to worry about setting back to dormant since the pickup is destroyed after being picked up.)
                ;; Note: Though the base pickup class has support to play FX on pickup in code, it requires an item definition which this pickup does not have. Thus handling it ourself in BP.
                (Variables|Default|SetCollectedByPlayerPawn _asfort_player_pawn_athena)
                (Networking|SetNetDormancy "DORM_Never")
                (Networking|ForceNetUpdate))))))
      (:Remote
        ;;[0.9599 0.3773 1] If this is a local pawn picking up the orb, allow them to play just the pickup fx early. This allows the pickup to feel more responsive on the client without risking the orb being invisible in the edge case where the player pawn actually didnt collide with it on the server.
        (if (and (HelperFunctions|IsLocalPawn _asfort_player_pawn_athena) (not (Variables|Default|GetbLocalFXPlayed)))
          (CallFunction|PlayLocalPickupFX)
          (Variables|Default|SetbLocalFXPlayed true))))))

(event Custom|TimeOrbDestroyed
  (params (DestroyedActor :type Actor%20Object%20Reference))
  (bind _touchcapsule (Variables|Collision|GetTouchCapsule))
  (bind _returnvalue (Transformation|GetWorldLocation _touchcapsule))
  ;[1 0.03332 0.8853] Play FX when a time orb expires naturally
  (if (not (Variables|Default|GetbTouched))
    (Networking|CanExecuteCosmeticEvents
      (:True
        (Development|PrintString "TimeOrb Expired Naturally")
        (Audio|PlaySoundatLocation (Variables|Sound|GetTimeOrb_Despawn) _returnvalue)
        (Effects|Components|ParticleSystem|SpawnFXAtLocation (Variables|Default|GetPickupFX) _returnvalue)
        (CallFunction|StopTimeOrbIdleLoopSFX)))))

(fn OnRep_CollectedByPlayerPawn
  ;[0.02871 0.7961 1] Play Pickup FX
  (bind _collectedbyplayerpawn (Variables|Default|GetCollectedByPlayerPawn)
    (:then
      (Variables|Default|SetbTouched true)
      (Rendering|SetActorHiddenInGame true)
      (Networking|CanExecuteCosmeticEvents
        (:True
          (CallFunction|StopTimeOrbIdleLoopSFX)
          (if (HelperFunctions|IsLocalPawn _collectedbyplayerpawn)
            (if (not (Variables|Default|GetbLocalFXPlayed))
              (CallFunction|PlayLocalPickupFX) ; If the player who collected this orb is a local player \nand they haven't already played their pickup FX, do so
            )
            (else
              (CallFunction|PlayRemotePickupFX) ; Play pickup FX for players who were not the one who collected the orb
            )))))
    (:else)))

(fn PlayLocalPickupFX
  (bind _rootcomponent (Variables|Transformation|GetRootComponent))
  (Effects|Components|ParticleSystem|SpawnFXAtLocation (Variables|Default|GetPickupFX) (Transformation|GetWorldLocation (Variables|Collision|GetTouchCapsule)))
  (Audio|SpawnSoundAttached (Variables|Sound|GetTimeOrb_Collect_1P) _rootcomponent)
  (Audio|SpawnSoundAttached (Variables|Sound|GetAlien%20VO%20Time%20Orb%20Collected%201P) _rootcomponent))

(fn PlayRemotePickupFX
  (Effects|Components|ParticleSystem|SpawnFXAtLocation (Variables|Default|GetPickupFX) (Transformation|GetWorldLocation (Variables|Collision|GetTouchCapsule)))
  (Audio|SpawnSoundAttached (Variables|Sound|GetTimeOrb_Collect_3P) (Variables|Transformation|GetRootComponent)))

(fn Stop_TimeOrb_IdleLoop_SFX
  (Audio|Components|Audio|FadeOut (Variables|Sound|GetTimeOrb_Idle_Loop_Comp) (Variables|Default|GetTime%20Orb%20Fade%20Out%20Duration) 0.0 "Sin" :!class /Script/Engine.AudioComponent))

(fn ConstructionScript)
