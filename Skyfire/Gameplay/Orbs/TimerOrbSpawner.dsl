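; Skyfire actor that spawns Time Orb pickups at tagged points, with server authority, during the CoreGameplay phase.
; After binding to the Skyfire mutator, it runs timed waves sized by mothership player count and playlist, with wave frequency varying by spawn phase.
; Shuffled points respect a reuse cooldown, and pickups return theirs to the pool.
; Timers or a squad reaching max loot rarity stop spawning and can delete active orbs.
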(event EventBeginPlay
  ;[0.1402 1 0.7264] BeginPlay - start waiting for mutator
  (|CallParentFunction)
  (CallFunction|WaitforMutator))

(event Custom|WaitForMutator
  (bind _returnvalue (Fortnite|Athena|GameState|GetGameStateAthena))
  ;[0.142 0.8701 1] WaitForMutator - Wait for and bind to the skyfire mutator
  (Utilities|IsValid _returnvalue
    (:Is%20Valid
      (bind _asfort_athena_mutator_skyfire (Utilities|Casting|CastToFortAthenaMutator_Skyfire (Modifier|GetMutatorbyClass (Variables|/Script/FortniteGame.FortGameStateAthena|GetMutatorListComponent _returnvalue) "/Script/SkyfireRuntime.FortAthenaMutator_Skyfire"))
        (:then
          (bind _output_get (Variables|Default|SetSkyfireMutator _asfort_athena_mutator_skyfire))
          (EventDispatchers|BindEventtoOnSkyfirePhaseChanged _output_get (EventDispatchers|CreateEvent :!event Custom|OnSkyfirePhaseChanged))
          (Networking|SwitchHasAuthority
            (:Authority
              (CallFunction|DeterminePlaylistType))))
        (:CastFailed
          (Utilities|Time|SetTimerbyEvent (delegate Custom|WaitForMutator) 0.1) ; Wait and retry
        )))))

(event Custom|SpawnTimeOrbs
  (params (OrbsToSpawn :type Integer))
  (bind _securityorbsspawnpointsunused_map (Variables|Default|GetSecurityOrbsSpawnPointsUnused_Map))
  ;[1 0.8838 0] SpawnTimeOrbs
  (Networking|SwitchHasAuthority
    (:Authority
      ;[0.142 0.8701 1] WaitForMutator - Wait for and bind to the skyfire mutator
      (Development|PrintString (Utilities|String|Append "orbs spawned: " (Utilities|String|ToString(Integer) OrbsToSpawn)) true true "(R=0.000000,G=0.660000,B=1.000000,A=1.000000)" 1e+01)
      (Variables|Default|SetTotalOrbsSpawned 0)
      (bind _keys (Utilities|Map|Keys (Variables|Default|GetSecurityOrbsSpawnPointsUnused_Map)))
      (Utilities|Array|Shuffle _keys) ; So each spawn wave we don't always choose the same spawn points
      (bind _array_element (Utilities|Array|ForEachLoopwithBreak _keys)
        (:LoopBody
          (if (and (< (Variables|Default|GetTotalOrbsSpawned) OrbsToSpawn) (> (Utilities|Array|Length _keys) 0))
            _
            (else
              _))
          (if (.ReturnValue (Utilities|Map|Find _securityorbsspawnpointsunused_map _array_element))
            (if (>= (- (Utilities|Time|GetTimeSeconds) (Utilities|Map|Find _securityorbsspawnpointsunused_map _array_element)) (ScalableFloat|GetValueatLevel (Variables|Default|GetSF_SpawnLocationCD)))
              (bind _returnvalue (Game|SpawnActorfromClass (Variables|Default|GetTimeOrbClass) (Math|Transform|MakeTransform (+ _array_element (Variables|Default|GetSpawnOffset))) "AdjustIfPossibleButAlwaysSpawn" "SelectDefaultAtRuntime" 0 0)) ; Make a struct of transform and time orb? So if time orb is consumed that space frees up
              (bind _astime_orb (Utilities|Casting|CastToTimeOrb _returnvalue))
              (Variables|/Skyfire/Gameplay/Orbs/TimeOrb.TimeOrb_C|SetTimerOrbSpawner self _astime_orb)
              (Utilities|Array|Add (Variables|Default|GetActiveTimeOrbs) _astime_orb)
              (Utilities|Map|Add (Variables|Default|GetUsedTimeOrbToTransformLocationTmap) _astime_orb _array_element)
              (Utilities|Map|Remove (Variables|Default|GetSecurityOrbsSpawnPointsUnused_Map) _array_element)
              (Variables|Default|SetTotalOrbsSpawned (+ (Variables|Default|GetTotalOrbsSpawned) 1)) ; Track the amount of spawned orbs.
            )))
        (:Completed
          _))
      (CallFunction|UpdateOrbCount (Variables|Default|GetTotalOrbsSpawned)))))

(event Custom|OnSkyfirePhaseChanged
  (params (NewPhase :type ESkyfirePhase%20Enum))
  ;[0.2833 1 0.4085] OnSkyfirePhaseChanged - On Prison phase, set spawn points and spawn initial orbs
  (Networking|SwitchHasAuthority
    (:Authority
      (switch Utilities|FlowControl|Switch|SwitchonESkyfirePhase NewPhase
        (:CoreGameplay
          (CallFunction|SetSpawnPoints)
          (Utilities|Time|SetTimerbyEvent (EventDispatchers|CreateEvent :!event Custom|SpawnInitialTimeOrbs) (ScalableFloat|GetValueatLevel (Variables|Default|GetInitialOrbSpawnDelay)) false false (ScalableFloat|GetValueatLevel (Variables|Default|GetTimeToLoadFromPrisonToGamespace)))
          (bind _returnvalue (Utilities|Time|SetTimerbyEvent (EventDispatchers|CreateEvent :!event Custom|StopSpawningOrbs) (ScalableFloat|GetValueatLevel (Variables|Default|GetTimeUntilStopSpawningOrbs))))
          (Variables|Default|SetStopSpawningOrbsTimer _returnvalue)
          (bind _returnvalue_2 (Utilities|Time|SetTimerbyEvent (EventDispatchers|CreateEvent :!event Custom|DeleteActiveOrbs) (ScalableFloat|GetValueatLevel (Variables|Default|GetTimeUntilDeleteAllOrbs))))
          (Variables|Default|SetDeleteActiveOrbsTimer _returnvalue_2)
          (bind _skyfiremutator (Variables|Default|GetSkyfireMutator)
            (:then
              (Playspace|BindEventtoOnFirstSquadReachedMaxLootRarity _skyfiremutator (EventDispatchers|CreateEvent :!event Custom|TeamReachedMaxLootTier)))
            (:else)))
        (:AllPlayersExited
          (CallFunction|CleanupTimers))))
    (:Remote
      (switch Utilities|FlowControl|Switch|SwitchonESkyfirePhase NewPhase
        (:CoreGameplay
          (Utilities|Time|SetTimerbyEvent (EventDispatchers|CreateEvent :!event TryToPlayTimeOrbSound) (ScalableFloat|GetValueatLevel (Variables|Default|GetInitialOrbSpawnDelay)) false false (ScalableFloat|GetValueatLevel (Variables|Default|GetTimeToLoadFromPrisonToGamespace))))))))

(event Custom|SpawnAdditionalOrbs
  (bind (_value _returnvalue) (Utilities|Map|Find (Variables|Default|GetTimeOrbWaveFrequencyPerPlaylistTypeTMap) (Variables|Default|GetPlaylistType)))
  ;[0.2274 1 0.9502] SpawnAdditionalOrbs
  (CallFunction|SpawnTimeOrbs (CallFunction|GetTotalOrbstoSpawnforPlayerCount))
  (bind _returnvalue_2 (Utilities|Time|SetTimerbyEvent (delegate Custom|SpawnAdditionalOrbs) (ScalableFloat|GetValueatLevel (select _returnvalue _value (Variables|Default|GetDefaultTimeOrbWaveFrequency)) (Math|Conversions|ToFloat(Integer) (Variables|Default|GetCurrentSpawnPhase)))))
  (Variables|Default|SetSpawnAdditionalOrbsTimer _returnvalue_2))

(event Custom|OnTimeOrbPickedUp
  (params (TimeOrb :type Time%20Orb%20Object%20Reference))
  (bind _usedtimeorbtotransformlocationtmap (Variables|Default|GetUsedTimeOrbToTransformLocationTmap))
  (bind (_value _returnvalue) (Utilities|Map|Find _usedtimeorbtotransformlocationtmap TimeOrb))
  ;[0.07955 0.9203 1] OnTimeOrbPickedUp - Add the location and time of the picked up time orb back to the list of available spawn locations. Time is used to determine if the location has been available long enough
  (CallFunction|UpdateOrbCount -1)
  (if _returnvalue
    (Utilities|Map|Add (Variables|Default|GetSecurityOrbsSpawnPointsUnused_Map) _value (Utilities|Time|GetTimeSeconds))
    (Utilities|Map|Remove _usedtimeorbtotransformlocationtmap TimeOrb)
    (Utilities|Array|RemoveItem (Variables|Default|GetActiveTimeOrbs) TimeOrb)))

(event Custom|StopSpawningOrbs
  ;[0.07832 0.9239 1] Stop Spawning Time Orbs
  (Utilities|Time|ClearandInvalidateTimerbyHandle (Variables|Default|GetSpawnAdditionalOrbsTimer)))

(event Custom|DeleteActiveOrbs
  (bind _activetimeorbs (Variables|Default|GetActiveTimeOrbs))
  ;[0.5625 0.298 0.1012] Delete All Remaining Time Orbs
  (for _array_element _activetimeorbs
    (Utilities|IsValid _array_element
      (:Is%20Valid
        (Actor|DestroyActor _array_element))))
  (Utilities|Array|Clear _activetimeorbs))

(event Custom|SpawnInitialTimeOrbs
  (bind (_value _returnvalue) (Utilities|Map|Find (Variables|Default|GetTimeOrbWaveFrequencyPerPlaylistTypeTMap) (Variables|Default|GetPlaylistType)))
  (bind (_value_2 _returnvalue_2) (Utilities|Map|Find (Variables|Default|GetWavePhaseDurationPerPlaylistTypeTMap) (Variables|Default|GetPlaylistType)))
  ;[0.7646 1 0] Spawn the initial set of time orbs.
  (CallFunction|SpawnTimeOrbs (CallFunction|GetTotalOrbstoSpawnforPlayerCount))
  (Variables|Default|SetInitial%20Spawn false)
  (bind _returnvalue_3 (Utilities|Time|SetTimerbyEvent (EventDispatchers|CreateEvent :!event Custom|SpawnAdditionalOrbs) (ScalableFloat|GetValueatLevel (select _returnvalue _value (Variables|Default|GetDefaultTimeOrbWaveFrequency)))))
  (Variables|Default|SetSpawnAdditionalOrbsTimer _returnvalue_3)
  (Utilities|Time|SetTimerbyEvent (EventDispatchers|CreateEvent :!event Custom|AdjustSpawnPhase) (ScalableFloat|GetValueatLevel (select _returnvalue_2 _value_2 (Variables|Default|GetDefaultWavePhaseDuration)))))

(event Custom|AdjustSpawnPhase
  (bind _playlisttype (Variables|Default|GetPlaylistType))
  (bind _wavephasedurationperplaylisttypetmap (Variables|Default|GetWavePhaseDurationPerPlaylistTypeTMap))
  (bind (_value _returnvalue) (Utilities|Map|Find _wavephasedurationperplaylisttypetmap _playlisttype))
  (bind _defaultwavephaseduration (Variables|Default|GetDefaultWavePhaseDuration))
  (bind _returnvalue_2 (select _returnvalue _value _defaultwavephaseduration))
  ;[1 0.2999 0.8039] Adjust the spawn phase durations
  (bind _output_get (Variables|Default|SetCurrentSpawnPhase (+ (Variables|Default|GetCurrentSpawnPhase) 1)))
  (if (not (< (ScalableFloat|GetValueatLevel _returnvalue_2 (Math|Conversions|ToFloat(Integer) _output_get)) 0.0))
    (Utilities|Time|SetTimerbyEvent (delegate Custom|AdjustSpawnPhase) (ScalableFloat|GetValueatLevel _returnvalue_2 (Math|Conversions|ToFloat(Integer) _output_get)))))

(event Custom|TeamReachedMaxLootTier
  ;[0.124 0.6299 1] Disable Time Orb spawning when a team reaches the max loot tier. Can also delete any active time orbs too if desired.
  (if (ScalableFloat|GetValueasBool (Variables|Default|GetDeleteAndStopSpawningTimeOrbs))
    (CallFunction|DeleteActiveOrbs)
    (Utilities|Time|ClearandInvalidateTimerbyHandle (Variables|Default|GetDeleteActiveOrbsTimer))
    (else
      _))
  (CallFunction|StopSpawningOrbs)
  (Utilities|Time|ClearandInvalidateTimerbyHandle (Variables|Default|GetStopSpawningOrbsTimer))
  (bind _skyfiremutator (Variables|Default|GetSkyfireMutator)
    (:then
      (Playspace|UnbindEventfromOnFirstSquadReachedMaxLootRarity _skyfiremutator (delegate Custom|TeamReachedMaxLootTier)))
    (:else)))

(event EventEndPlay
  (params (EndPlayReason :type EEndPlayReason%20Enum :default "Destroyed"))
  ;[0.1149 0.6081 1] Clear and Invalidate timers when time orb spawner is destroyed
  (|CallParentFunction EndPlayReason)
  (Utilities|Time|ClearandInvalidateTimerbyHandle (Variables|Default|GetStopSpawningOrbsTimer))
  (Utilities|Time|ClearandInvalidateTimerbyHandle (Variables|Default|GetSpawnAdditionalOrbsTimer))
  (bind _skyfiremutator (Variables|Default|GetSkyfireMutator)
    (:then
      (EventDispatchers|UnbindEventfromOnSkyfirePhaseChanged _skyfiremutator (EventDispatchers|CreateEvent :!event Custom|OnSkyfirePhaseChanged)))
    (:else)))

(event Custom|CleanupTimers
  ;[0.1149 0.6081 1] Clear and Invalidate timers when time orb spawner is destroyed
  (Utilities|Time|ClearandInvalidateTimerbyHandle (Variables|Default|GetStopSpawningOrbsTimer))
  (Utilities|Time|ClearandInvalidateTimerbyHandle (Variables|Default|GetSpawnAdditionalOrbsTimer))
  (bind _skyfiremutator (Variables|Default|GetSkyfireMutator)
    (:then
      (EventDispatchers|UnbindEventfromOnSkyfirePhaseChanged _skyfiremutator (EventDispatchers|CreateEvent :!event Custom|OnSkyfirePhaseChanged)))
    (:else)))

(fn DeterminePlaylistType
  (bind _returnvalue (Fortnite|Athena|GameState|GetGameStateAthena))
  ;[0.8353 1 0] Get the playlist tag to determine if it's solo/squad etc.
  (Utilities|IsValid _returnvalue
    (:Is%20Valid
      (bind _array_element (Utilities|Array|ForEachLoopwithBreak (Variables|Default|GetPossiblePlaylistTags))
        (:LoopBody
          (if (GameplayTags|TagContainer|CompareTagContainertoOtherTags _array_element (Class|FortGameStateAthena|GetAthenaPlaylistContextTags _returnvalue))
            (Variables|Default|SetPlaylistType _array_element)))))))

(fn GetTotalOrbsToSpawnForPlayerCount
  (bind _initial_spawn (Variables|Default|GetInitial%20Spawn))
  (bind _playlisttype (Variables|Default|GetPlaylistType))
  (bind _initialorbsperplaylisttypetmap (Variables|Default|GetInitialOrbsPerPlaylistTypeTMap))
  (bind (_value _returnvalue) (Utilities|Map|Find _initialorbsperplaylisttypetmap _playlisttype))
  (bind _initialorbsperplayer (Variables|Default|GetInitialOrbsPerPlayer))
  (bind _returnvalue_2 (select _returnvalue _value _initialorbsperplayer))
  (bind _returnvalue_3 (ScalableFloat|GetValueatLevel _returnvalue_2))
  (bind _additionalorbsperplayer (Variables|Default|GetAdditionalOrbsPerPlayer))
  (bind _returnvalue_4 (ScalableFloat|GetValueatLevel _additionalorbsperplayer))
  (bind _returnvalue_5 (select _initial_spawn _returnvalue_3 _returnvalue_4))
  (bind _returnvalue_6 (* (Class|FortAthenaMutatorSkyfire|GetNumberOfPlayersinMothership (Variables|Default|GetSkyfireMutator) false false) _returnvalue_5))
  (bind (_value_2 _returnvalue_7) (Utilities|Map|Find (Variables|Default|GetMinimumInitialOrbsPerPlaylistTypeTMap) (Variables|Default|GetPlaylistType)))
  (bind _skyfiremutator (Variables|Default|GetSkyfireMutator))
  (bind _returnvalue_8 (Class|FortAthenaMutatorSkyfire|GetNumberOfPlayersinMothership _skyfiremutator false false))
  ;[0.02166 0.6963 1] Active player count * orbs per player = total orbs to spawn
  (Utilities|IsValid _skyfiremutator
    (:Is%20Valid
      (if (Variables|Default|GetInitial%20Spawn)
        (return (Math|Float|Ceil (Math|Float|Max(Float) _returnvalue_6 (ScalableFloat|GetValueatLevel (select _returnvalue_7 _value_2 (Variables|Default|GetDefaultMinimumInitialOrbsPerPlaylistType))))))
        (else
          (return (Math|Float|Ceil _returnvalue_6)))))))

(fn SetSpawnPoints
  ;[0.1961 0.5817 1] SetSpawnPoints - Set the spawn point locations
  (bind _outactors (Actor|GetAllActorswithTag (Variables|Default|GetTimeOrbSpawnerTag)))
  (for _array_element _outactors
    (Utilities|Array|Add (Variables|Default|GetSecurityOrbsSpawnPointsMaster) (Transformation|GetActorLocation _array_element))
    (Utilities|Map|Add (Variables|Default|GetSecurityOrbsSpawnPointsUnused_Map) (Transformation|GetActorLocation _array_element))))

(fn TryToPlayTimeOrbSound
  (bind _returnvalue (Fortnite|GetFirstLocalFortPlayerController true true))
  (bind _returnvalue_2 (Game|Player|GetPlayerPawn _returnvalue))
  (bind _returnvalue_3 (Actor|GetComponentbyClass _returnvalue_2 "/Script/SkyfireRuntime.SkyfirePlayerPawnComponent"))
  (Utilities|IsValid _returnvalue_2
    (:Is%20Valid
      (Utilities|IsValid _returnvalue_3
        (:Is%20Valid
          (if (Class|SkyfirePlayerPawnComponent|IsinMothership _returnvalue_3)
            (Audio|PlaySound2D (Variables|Default|GetSFX_VO_TimeOrb_Appear_FirstTime))))))))

(fn UpdateOrbCount
  (params (Amount :type Integer))
  (bind _skyfiremutator (Variables|Default|GetSkyfireMutator)
    (:then
      (Playspace|ModifyTimeOrbCount _skyfiremutator Amount))
    (:else)))

(fn ConstructionScript)
