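; Test-map time orb spawner that places an orb on begin play and spawns a replacement whenever that orb is destroyed.
; After SetSpawnPoints, SpawnTimeOrbs runs on authority only, taking a position from the unused spawn-point map's keys plus an offset.
; A toggled boolean converted to an array index alternates between two locations, and each orb's OnDestroyed event triggers the next spawn.
; The OrbsToSpawn parameter is unused, so each call spawns one orb.
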(event EventBeginPlay
  ;[0 1 0.1608] Setup the spawning.
  (CallFunction|SetSpawnPoints)
  (CallFunction|SpawnTimeOrbs 1)
  (Variables|Default|SetInitial%20Spawn false))

(event EventSpawnTimeOrbs
  (params (OrbsToSpawn :type Integer))
  ;[0 0.128 1] Determine which of the 2 spawn locations we should use and spawn a time orb. Bind to the orb being collected to spawn the next one.
  (Networking|SwitchHasAuthority
    (:Authority
      (bind _keys (Utilities|Map|Keys (Variables|Default|GetSecurityOrbsSpawnPointsUnused_Map)))
      (bind _returnvalue (Game|SpawnActorfromClass (Variables|Default|GetTimeOrbClass) (Math|Transform|MakeTransform (+ (Utilities|Array|Get(acopy) _keys (Math|Conversions|ToInteger(Boolean) (Variables|Default|GetSpawnPoint1Taken))) (Variables|Default|GetSpawnOffset))) "AdjustIfPossibleButAlwaysSpawn" "SelectDefaultAtRuntime" 0 0)) ; Make a struct of transform and time orb? So if time orb is consumed that space frees up
      (Game|BindEventtoOnDestroyed _returnvalue (EventDispatchers|CreateEvent :!event Custom|OrbPickedUp))
      (Variables|Default|SetSpawnPoint1Taken (not (Variables|Default|GetSpawnPoint1Taken))))))

(event Custom|OrbPickedUp
  (params (DestroyedActor :type Actor%20Object%20Reference))
  ;[0.714 0.1882 1] Spawn another orb.
  (CallFunction|SpawnTimeOrbs 1))

(fn ConstructionScript)
